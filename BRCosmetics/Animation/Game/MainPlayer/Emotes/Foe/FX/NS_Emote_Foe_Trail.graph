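# Niagara system for the Foe emote that draws a glowing green trail around the emote's pumpkin gadget.
# A movement-spawned ribbon forms the tapering trail, while a burst sprite with a light renderer adds a fading glow.
# Two soft sprite layers, one GPU flipbook and one CPU with animated material parameters, plus short-lived sparks sample the skeletal mesh's JNT_Pumpkin bone and inherit velocity.
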NiagaraSystem /BRCosmetics/Animation/Game/MainPlayer/Emotes/Foe/FX/NS_Emote_Foe_Trail
dependencies {
 usedDataInterfaces [&NiagaraDataInterfaceCurve, &NiagaraDataInterfaceColorCurve, &NiagaraDataInterfaceVector2DCurve]
 usedDynamicInputs []
 usedModules [&SystemState]
}
emitters {
 Empty {
  emitterUpdateScript [
   &EmitterState(Inactive Response=::NewEnumerator1, Life Cycle Mode=::NewEnumerator1, Loop Duration=1, UseLoopDelay=false)
   &SpawnRate(SpawnRate=45, Use Spawn Probability=false) {enabled=false}
   &SpawnPerUnit(Module.Use Max Movement Threshold=true, Module.Use Movement Tolerance=false, Spawn Spacing=3, Use Spawn Probability=false)
  ]
  particleSpawnScript [&InitializeRibbon(Color=<rgba 1 1 1 2>, Lifetime=0.78, Ribbon Width=17, UsePositionOffset=false, Write Material Random=true)]
  particleUpdateScript [
   &ParticleState()
   &ScaleRibbonWidth(Ribbon Width Scale=&FloatFromCurve(FloatCurve=DataInterface() {propertyValues {Curve {keys [
         {arriveTangent=10.6409, interpMode=RCIM_Cubic, leaveTangent=7.4546, tangentMode=RCTM_Break, time=0, value=0.5}
         {interpMode=RCIM_Cubic, leaveTangent=0.0729, tangentMode=RCTM_Break, time=0.1, value=1}
         {arriveTangent=-0.3829, interpMode=RCIM_Cubic, leaveTangent=-0.3829, tangentMode=RCTM_User, time=0.3, value=1}
         {arriveTangent=-2.0489, interpMode=RCIM_Cubic, leaveTangent=-2.0489, tangentMode=RCTM_User, time=1, value=0}
        ]}}}))
   &ScaleColor(
    Linear Color Curve=DataInterface() {propertyValues {
      AlphaCurve {keys [<at 0 1>,<at 0.2985 1>,<at 0.5735 0.5>,<at 1 0>]}
      BlueCurve {keys [<at 0 0>,<at 0.1031 0>,<at 0.3817 0.1139>,<at 1 0.0677>]}
      GreenCurve {keys [<at 0 2>,<at 0.1031 0.8438>,<at 0.3817 0.3021>,<at 1 0.1458>]}
      RedCurve {keys [<at 0 0.3354>,<at 0.1031 0.0304>,<at 0.3817 0>,<at 1 0.0017>]}
     }}
    Scale Mode=::NewEnumerator2
   )
  ]
  rendererClasses [&NiagaraRibbonRendererProperties]
  renderers [{rendererClass=&NiagaraRibbonRendererProperties}]
 }
 Empty001 {
  emitterUpdateScript [
   &EmitterState(Life Cycle Mode=::NewEnumerator1, Loop Behavior=::NewEnumerator1, Loop Duration Mode=::NewEnumerator1, UseLoopDelay=false)
   &SpawnRate(SpawnRate=20, Use Spawn Probability=false)
  ]
  particleSpawnScript [
   &InitializeParticle(
    Color=&RandomRangeLinearColor(Minimum=<rgba 0.0127 0.2292 0.0539 0.15>, Maximum=<rgba 0.0789 0.4375 0 0.4>)
    Lifetime Max=1.8
    Lifetime Mode=::NewEnumerator1
    Sprite Rotation Angle=&RandomRangeFloat(Minimum=0, Maximum=1)
    Sprite Rotation Mode=::NewEnumerator1
    Sprite Size Mode=::NewEnumerator1
    Uniform Sprite Size Max=40
    Uniform Sprite Size Min=15
    UsePositionOffset=false
    Write Material Random=true
   )
   &CameraOffset(Camera Offset Amount=6)
   &SkeletalMeshLocation(
    Bone Sampling Mode=::NewEnumerator4
    MaskSpawnGroup=false
    Skeletal Mesh=DataInterface() {propertyValues {
      DefaultMesh=null
      ExcludeBoneName=root
      OptionalComponentClass=null
      PreviewMesh=/BRCosmetics/Animation/Props/Gadgets_Foe/Meshes/Foe_Gadget
      SkinningMode=SkinOnTheFly
      SoftSourceActor=null
      SourceMode=Default
      UvSetIndex=0
      WholeMeshLOD=-1
      bExcludeBone=false
      bReadDeformedGeometry=true
      bRequireCurrentFrameData=true
      ComponentTags []
      FilteredBones [JNT_Pumpkin]
      FilteredSockets []
      MeshUserParameter {parameter {name=None, typeDefHandle {registeredTypeIndex=1}, varData []}}
      SamplingRegions []
     }}
   )
   &SphereLocation(
    MaskSpawnGroup=false
    Override Seed=false
    Sphere Radius=&RandomRangeFloat(Minimum=2, Maximum=10)
    Surface Only=false
    UseNonUniformScale=false
    UseOffset=false
    UseSphereOrientation=false
   )
   &AddVelocity(Velocity=&RandomRangeVector(Minimum=<-15 -15 25>, Maximum=<15 15 40>))
   &AddVelocityFromPoint(Module.Use Falloff Distance=false, Velocity Strength=&RandomRangeFloat(Minimum=14, Maximum=35))
   &InheritVelocity(
    Inherited Velocity Speed Limit:NiagaraFloat=&RandomRangeFloat(Minimum=-50, Maximum=-100)
    Speed Limit=true
    UseSpeedThreshold=false
   )
  ]
  particleUpdateScript [
   &Drag(Drag=&RandomRangeFloat(Minimum=3.5, Maximum=5), Use Linear Drag=true, Use Rotational Drag=false)
   &SolveForcesAndVelocity(Clamp Velocity=false, Limit Acceleration=false)
   &ParticleState()
   &ScaleColor(
    Linear Color Curve=DataInterface() {propertyValues {
      AlphaCurve {keys [<at 0 0>,<at 0.1317 1>,<at 0.3996 1>,<at 1 0>]}
      BlueCurve {keys [<at 0 1>,<at 0.1411 1>,<at 0.5192 0.3802>,<at 1 0.2396>]}
      GreenCurve {keys [<at 0 1>,<at 0.1411 1>,<at 0.5192 0.3802>,<at 1 0.2396>]}
      RedCurve {keys [<at 0 1>,<at 0.1411 1>,<at 0.5192 0.3802>,<at 1 0.2396>]}
     }}
    Scale Mode=::NewEnumerator2
   )
   &ScaleSpriteSize(Uniform Curve Sprite Scale=DataInterface() {propertyValues {Curve {keys [
        {arriveTangent=1.4197, interpMode=RCIM_Cubic, leaveTangent=1.4197, tangentMode=RCTM_User, time=0, value=0.7}
        {interpMode=RCIM_Cubic, tangentMode=RCTM_User, time=1, value=1.25}
       ]}}})
   &DynamicMaterialParameters(
    Index 0 Param 1=&FloatFromCurve(FloatCurve=DataInterface() {propertyValues {Curve {keys [
         {arriveTangent=-2.2229, interpMode=RCIM_Cubic, leaveTangent=-2.2229, tangentMode=RCTM_User, time=0, value=1}
         {arriveTangent=0.1268, interpMode=RCIM_Cubic, leaveTangent=0.1268, tangentMode=RCTM_User, time=1, value=0}
        ]}}})
    Index 0 Param 2=1
    Index 0 Param 3=1
    Index 0 Param 4=1
    Index 2 Param 1=1
    Index 2 Param 2=1
    Index 2 Param 3=&FloatFromCurve(FloatCurve=DataInterface() {propertyValues {Curve {keys [
         {arriveTangent=1.6595, interpMode=RCIM_Cubic, leaveTangent=1.6595, tangentMode=RCTM_User, time=0.5, value=1}
         {arriveTangent=7.6852, interpMode=RCIM_Cubic, leaveTangent=7.6852, tangentMode=RCTM_User, time=1, value=3}
        ]}}})
    Index 2 Param 4=1
    Module.Param0WriteEnabled=true
    Module.Param0WriteEnabled2=true
    Module.Param1WriteEnabled=true
    Module.Param1WriteEnabled2=true
    Module.Param2WriteEnabled=true
    Module.Param2WriteEnabled2=true
    Module.Param3WriteEnabled=true
    Module.Param3WriteEnabled2=true
    Write Parameter Index 2=true
   )
  ]
  rendererClasses [&NiagaraSpriteRendererProperties]
  renderers [{rendererClass=&NiagaraSpriteRendererProperties}]
 }
 Empty002 {
  simTarget=GPUComputeSim
  emitterUpdateScript [
   &EmitterState(Life Cycle Mode=::NewEnumerator1, Loop Duration=1, Scalability Mode=::NewEnumerator1, UseLoopDelay=false)
   &SpawnRate(SpawnRate=14, Use Spawn Probability=false)
  ]
  particleSpawnScript [
   &InitializeParticle(
    Color=&RandomRangeLinearColor(Minimum=<rgba 0 0.8802 0.2146 0.3>, Maximum=<rgba 0.1522 0.8438 0 0.4>)
    Lifetime Max=1.5
    Lifetime Min=0.7
    Lifetime Mode=::NewEnumerator1
    Sprite Rotation Angle=&RandomRangeFloat(Minimum=0, Maximum=1)
    Sprite Rotation Mode=::NewEnumerator1
    Sprite Size Mode=::NewEnumerator1
    Uniform Sprite Size Max=55
    Uniform Sprite Size Min=26
    UsePositionOffset=false
    Write Material Random=true
   )
   &CameraOffset(Camera Offset Amount=6)
   &SkeletalMeshLocation(
    Bone Sampling Mode=::NewEnumerator4
    MaskSpawnGroup=false
    Skeletal Mesh=DataInterface() {propertyValues {
      DefaultMesh=null
      ExcludeBoneName=root
      OptionalComponentClass=null
      PreviewMesh=&Foe_Gadget
      SkinningMode=SkinOnTheFly
      SoftSourceActor=null
      SourceMode=Default
      UvSetIndex=0
      WholeMeshLOD=-1
      bExcludeBone=false
      bReadDeformedGeometry=true
      bRequireCurrentFrameData=true
      ComponentTags []
      FilteredBones [JNT_Pumpkin]
      FilteredSockets []
      MeshUserParameter {parameter {name=None, typeDefHandle {registeredTypeIndex=1}, varData []}}
      SamplingRegions []
     }}
   )
   &SphereLocation(
    MaskSpawnGroup=false
    Override Seed=false
    Sphere Radius=4
    Surface Only=false
    UseNonUniformScale=false
    UseOffset=false
    UseSphereOrientation=false
   )
   &AddVelocity(Velocity=&RandomRangeVector(Minimum=<-20 -20 25>, Maximum=<20 20 40>))
   &AddVelocityFromPoint(Module.Use Falloff Distance=false, Velocity Strength=&RandomRangeFloat(Minimum=10, Maximum=25))
   &InheritVelocity(
    Inherited Velocity Speed Limit:NiagaraFloat=&RandomRangeFloat(Minimum=-85, Maximum=-160)
    Speed Limit=true
    UseSpeedThreshold=false
   )
  ]
  particleUpdateScript [
   &Drag(Drag=5, Use Linear Drag=true, Use Rotational Drag=false)
   &SolveForcesAndVelocity(Clamp Velocity=false, Limit Acceleration=false)
   &ParticleState()
   &SubUVAnimation(
    End Frame=30
    Start Frame=14
    SubUV Loop Count:NiagaraFloat=&FloatFromCurve(FloatCurve=DataInterface() {propertyValues {Curve {keys [<at 0 1 RCIM_Cubic>, {arriveTangent=-0.5641, interpMode=RCIM_Cubic, leaveTangent=-0.5641, tangentMode=RCTM_User, time=1, value=0.7}]}}})
   )
   &ScaleColor(
    Scale Alpha=&FloatFromCurve(FloatCurve=DataInterface() {propertyValues {Curve {keys [
         {interpMode=RCIM_Cubic, leaveTangent=14.0237, tangentMode=RCTM_Break, time=0, value=0}
         {interpMode=RCIM_Cubic, tangentMode=RCTM_Break, time=0.1, value=1}
         {arriveTangent=-2.3416, interpMode=RCIM_Cubic, leaveTangent=-1.342, tangentMode=RCTM_Break, time=0.5, value=0.7}
         {interpMode=RCIM_Cubic, tangentMode=RCTM_Break, time=1, value=0}
        ]}}})
    ScaleA=true
    ScaleRGB=true
   )
   &ScaleSpriteSize(Uniform Curve Sprite Scale=DataInterface() {propertyValues {Curve {keys [
        {arriveTangent=1.0474, interpMode=RCIM_Cubic, leaveTangent=1.0474, tangentMode=RCTM_User, time=0, value=0.55}
        {interpMode=RCIM_Cubic, tangentMode=RCTM_User, time=1, value=1}
       ]}}})
  ]
  rendererClasses [&NiagaraSpriteRendererProperties]
  renderers [{rendererClass=&NiagaraSpriteRendererProperties}]
 }
 Empty004 {
  emitterUpdateScript [
   &EmitterState(
    Inactive Response=::NewEnumerator1
    Life Cycle Mode=::NewEnumerator1
    Loop Behavior=::NewEnumerator1
    Loop Duration Mode=::NewEnumerator1
    UseLoopDelay=false
   )
   &SpawnRate(Use Spawn Probability=false) {enabled=false}
   &SpawnBurst_Instantaneous(Use Spawn Probability=false, UseLoopCountLimit=false)
  ]
  particleSpawnScript [
   &InitializeParticle(
    Color=<rgba 0.0222 0.3385 0.0267 0.45>
    Lifetime=5.2
    Sprite Size Mode=::NewEnumerator3
    Uniform Sprite Size=42
    UsePositionOffset=false
    Write Material Random=true
   )
   /SetVariables(Particles.Exponent=3, Particles.Radius=40) {isSetParameters=true}
   &CameraOffset(Camera Offset Amount=5)
  ]
  particleUpdateScript [
   &ParticleState()
   &ScaleColor(
    Scale Alpha=&FloatFromCurve(FloatCurve=DataInterface() {propertyValues {Curve {keys [
         {arriveTangent=9.0725, interpMode=RCIM_Cubic, leaveTangent=9.0725, tangentMode=RCTM_User, time=0, value=0}
         {arriveTangent=0.2352, interpMode=RCIM_Cubic, leaveTangent=0.2352, tangentMode=RCTM_User, time=0.1408, value=0.9847}
         {arriveTangent=-1.7863, interpMode=RCIM_Cubic, leaveTangent=-1.9303, tangentMode=RCTM_Break, time=0.5, value=0.55}
         {arriveTangent=-0.0405, interpMode=RCIM_Cubic, leaveTangent=-0.0405, tangentMode=RCTM_User, time=1, value=0}
        ]}}})
    ScaleA=true
    ScaleRGB=true
   )
   /SetVariables(
    Particles.LightColor=&ColorFromCurve(ColorCurve=DataInterface() {propertyValues {
       AlphaCurve {keys [<at 0 0 RCIM_Cubic>,<at 0.046 1>,<at 0.3631 1>,<at 1 0 RCIM_Cubic>]}
       BlueCurve {keys [
         <at 0.0013 0>
         {arriveTangent=1.7817, interpMode=RCIM_Cubic, leaveTangent=1.7817, time=0.0729, value=0}
         <at 0.2659 0.4715>
         <at 0.8054 0>
         <at 1 0 RCIM_Cubic>
        ]}
       GreenCurve {keys [
         <at 0.0013 0>
         {arriveTangent=30.2319, interpMode=RCIM_Cubic, leaveTangent=30.2319, time=0.0729, value=9}
         <at 0.2659 8>
         <at 0.8054 5>
         <at 1 0 RCIM_Cubic>
        ]}
       RedCurve {keys [<at 0.0013 0>,<at 0.0729 0.1523 RCIM_Cubic>,<at 0.2659 0>,<at 0.8054 0.1088>,<at 1 0 RCIM_Cubic>]}
      }})
    Particles.Position1:Vector3f=&Add_Vector(A=Linked() {linkedVariable {name=Particles.Position, type=NiagaraPosition}}, B=<0 8 0>)
   ) {isSetParameters=true}
  ]
  rendererClasses [&NiagaraSpriteRendererProperties, &NiagaraLightRendererProperties]
  renderers [
   {rendererClass=&NiagaraSpriteRendererProperties}
   {rendererClass=&NiagaraLightRendererProperties, rendererIndex=1}
  ]
 }
 Spark {
  emitterUpdateScript [
   &EmitterState(Life Cycle Mode=::NewEnumerator1, Loop Behavior=::NewEnumerator1, Loop Delay=0.085, Loop Duration Mode=::NewEnumerator1, UseLoopDelay=true)
   &SpawnRate(SpawnRate=22, Use Spawn Probability=false)
   &SpawnBurst_Instantaneous(Spawn Count=3, Use Spawn Probability=false, UseLoopCountLimit=false) {enabled=false}
  ]
  particleSpawnScript [
   &InitializeParticle(
    Color=<rgba 0.3578 7 0 1>
    Lifetime Max=0.3
    Lifetime Min=0.2
    Lifetime Mode=::NewEnumerator1
    Sprite Size Max=<0.8 4>
    Sprite Size Min=<0.5 2>
    Sprite Size Mode=::NewEnumerator2
    UsePositionOffset=false
    Write Material Random=true
   )
   &SphereLocation(
    MaskSpawnGroup=false
    Override Seed=false
    Sphere Radius=12
    Surface Only=false
    UseNonUniformScale=false
    UseOffset=false
    UseSphereOrientation=false
   ) {enabled=false}
   &AddVelocityFromPoint(Module.Use Falloff Distance=false, Velocity Strength=&RandomRangeFloat(Minimum=12, Maximum=30)) {enabled=false}
   &InheritVelocity(
    Inherited Velocity Speed Limit:NiagaraFloat=&RandomRangeFloat(Minimum=25, Maximum=10)
    Speed Limit=true
    UseSpeedThreshold=false
   )
   &SkeletalMeshLocation(
    Bone Sampling Mode=::NewEnumerator4
    MaskSpawnGroup=false
    Skeletal Mesh=DataInterface() {propertyValues {
      DefaultMesh=null
      ExcludeBoneName=root
      OptionalComponentClass=null
      PreviewMesh=&Foe_Gadget
      SkinningMode=SkinOnTheFly
      SoftSourceActor=null
      SourceMode=Default
      UvSetIndex=0
      WholeMeshLOD=-1
      bExcludeBone=false
      bReadDeformedGeometry=true
      bRequireCurrentFrameData=true
      ComponentTags []
      FilteredBones [JNT_Pumpkin]
      FilteredSockets []
      MeshUserParameter {parameter {name=None, typeDefHandle {registeredTypeIndex=1}, varData []}}
      SamplingRegions []
     }}
   )
   &SphereLocation(
    MaskSpawnGroup=false
    Override Seed=false
    Sphere Radius=3
    Surface Only=false
    UseNonUniformScale=false
    UseOffset=false
    UseSphereOrientation=false
   ) {module=SphereLocation001}
  ]
  particleUpdateScript [
   &Drag(Drag=2, Use Linear Drag=true, Use Rotational Drag=false)
   &SolveForcesAndVelocity(Clamp Velocity=false, Limit Acceleration=false)
   &ParticleState()
   &ScaleSpriteSize(
    Non-Uniform Curve Sprite Scale=DataInterface() {propertyValues {
      XCurve {keys [{tangentMode=RCTM_User, time=0, value=1}, {tangentMode=RCTM_User, time=1, value=0.2}]}
      YCurve {keys [{tangentMode=RCTM_User, time=0, value=1}, {tangentMode=RCTM_User, time=1, value=0}]}
     }}
    Scale Sprite Size Mode=::NewEnumerator3
   )
  ]
  rendererClasses [&NiagaraSpriteRendererProperties]
  renderers [{rendererClass=&NiagaraSpriteRendererProperties}]
 }
}
summary {
 systemName=NS_Emote_Foe_Trail
 emitters [
  {emitterName=Empty, rendererClasses [&NiagaraRibbonRendererProperties]}
  {emitterName=Empty002, simTarget=GPUComputeSim, rendererClasses [&NiagaraSpriteRendererProperties]}
  {emitterName=Empty001, rendererClasses [&NiagaraSpriteRendererProperties]}
  {emitterName=Empty004, rendererClasses [&NiagaraSpriteRendererProperties, &NiagaraLightRendererProperties]}
  {emitterName=Spark, rendererClasses [&NiagaraSpriteRendererProperties]}
 ]
 userVariables [User.SkeletalMesh {
   type=NiagaraDataInterfaceSkeletalMesh
   defaultValue {
    struct=NiagaraExt_VariableValue_DataInterface
    value {dataInterface=&self:NiagaraDataInterfaceSkeletalMesh_0}
   }
  }]
}
